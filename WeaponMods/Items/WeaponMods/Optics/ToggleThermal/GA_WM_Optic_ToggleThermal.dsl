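; Gameplay ability for a toggleable thermal weapon optic, granting a thermal vision effect while the player is targeting with it switched on.
; On activation it finds or adds a TargetModeSelector component on the current weapon, then binds targeting and mode-change events that apply or remove the effect.
; A RepNotify on bIsTargeting switches custom depth stencil values on weapon and mod meshes.
; Ending removes the effect and unbinds, deliberately keeping the component for re-equip.
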(event Ability|EventActivateAbility
  (|CallParentFunction)
  (bind _asfort_player_state_athena (Utilities|Casting|CastToFortPlayerStateAthena (Ability|GetOwningActorfromActorInfo))
    (:then
      (Utilities|IsValid (PlayerState|GetCurrentPawn _asfort_player_state_athena)
        (:Is%20Valid
          (Utilities|IsValid (Weapon|GetCurrentWeapon (PlayerState|GetCurrentPawn _asfort_player_state_athena))
            (:Is%20Valid
              (bind _output_get (Variables|Default|SetCurrentThermalWeapon (Weapon|GetCurrentWeapon (PlayerState|GetCurrentPawn _asfort_player_state_athena))))
              (Utilities|FlowControl|Sequence
                (:then_0
                  ;[0.3977 1 0.2188] Find or create target mode selector component
                  (bind _output_get_2 (Variables|Default|SetTargetModeSelectorComponent (Actor|GetComponentbyClass _output_get "/Script/FortniteGame.FortWeaponComponent_TargetModeSelector")))
                  (Utilities|IsValid _output_get_2
                    (:Is%20Not%20Valid
                      (bind _returnvalue (Game|AddComponentbyClass _output_get "/ToggleScopeModeInput/Gameplay/FortWeaponComponent_TargetModeSelector_Base.FortWeaponComponent_TargetModeSelector_Base_C"))
                      (bind _output_get_3 (Variables|Default|SetTargetModeSelectorComponent _returnvalue))
                      (Utilities|IsValid _output_get_3
                        (:Is%20Valid
                          (Class|FortWeaponComponentTargetModeSelector|ServerSetTargetMode _output_get_3 1))))))
                (:then_1
                  ;[0.2185 1 0.7253] Bind to targeting events
                  (EventDispatchers|BindEventtoOnTargetingChanged _output_get (EventDispatchers|CreateEvent :!event Custom|OnWeaponTargetingChanged))
                  (bind _targetmodeselectorcomponent (Variables|Default|GetTargetModeSelectorComponent)
                    (:then
                      (EventDispatchers|BindEventtoOnTargetModeChanged _targetmodeselectorcomponent (EventDispatchers|CreateEvent :!event Custom|OnTargetModeChanged))
                      (CallFunction|OnTargetModeChanged (Variables|/Script/FortniteGame.FortWeaponComponent_TargetModeSelector|GetTargetMode _targetmodeselectorcomponent)))
                    (:else
                      ;[1 0 0.9571] TEMP DEBUG LOGGING
                      (Development|PrintString "THERMALOptic: Target Selection Component Creation Failed" true true "(R=1.000000,G=0.000000,B=0.187196,A=1.000000)" 12.0))))))
            (:Is%20Not%20Valid
              ;[1 0 0.9571] TEMP DEBUG LOGGING
              (Development|PrintString "THERMALOptic: Weapon INVALID" true true "(R=1.000000,G=0.000000,B=0.187196,A=1.000000)" 12.0))))
        (:Is%20Not%20Valid
          (Development|PrintString "THERMALOptic: Player Pawn INVALID" true true "(R=1.000000,G=0.000000,B=0.187196,A=1.000000)" 12.0))))
    (:CastFailed
      (Development|PrintString "THERMALOptic: Player State cast FAILED" true true "(R=1.000000,G=0.000000,B=0.187196,A=1.000000)" 12.0))))

(event Custom|OnWeaponTargetingChanged
  (params (bIsTargeting :type Boolean))
  (bind _returnvalue (Ability|GetOwningActorfromActorInfo))
  (bind _returnvalue_2 (Ability|GetFortAbilitySystemComponent _returnvalue))
  ;[0.1739 1 0.3382] Enable / Disable Thermal
  (Variables|Default|SetbIsTargeting bIsTargeting)
  (if (Math|Boolean|ANDBoolean (Variables|Default|GetbIsTargeting) (Variables|Default|GetbIsToggled) (not (Variables|Default|GetbThermalActive)))
    (bind _returnvalue_3 (GameplayEffects|ApplyGameplayEffectToSelf :self _returnvalue_2 :GameplayEffectClass "/WeaponMods/Items/WeaponMods/Optics/ToggleThermal/GE_ThermalVision_Grant.GE_ThermalVision_Grant_C" :Level 0.0))
    (Variables|Default|SetGE_ThermalVision _returnvalue_3)
    (Variables|Default|SetbThermalActive true)
    (else
      (if (Variables|Default|GetbThermalActive)
        (GameplayEffects|RemoveActiveGameplayEffect _returnvalue_2 (Variables|Default|GetGE_ThermalVision) -1)
        (Variables|Default|SetGE_ThermalVision (Utilities|Struct|MakeActiveGameplayEffectHandle))
        (Variables|Default|SetbThermalActive false)))))

(event Custom|OnTargetModeChanged
  (params (TargetMode :type Integer))
  (bind _returnvalue (Ability|GetOwningActorfromActorInfo))
  (bind _returnvalue_2 (Ability|GetFortAbilitySystemComponent _returnvalue))
  ;[0.1739 1 0.3382] Enable / Disable Thermal
  (Variables|Default|SetbIsToggled (!= TargetMode 0))
  (if (Math|Boolean|ANDBoolean (Variables|Default|GetbIsTargeting) (Variables|Default|GetbIsToggled) (not (Variables|Default|GetbThermalActive)))
    (bind _returnvalue_3 (GameplayEffects|ApplyGameplayEffectToSelf :self _returnvalue_2 :GameplayEffectClass "/WeaponMods/Items/WeaponMods/Optics/ToggleThermal/GE_ThermalVision_Grant.GE_ThermalVision_Grant_C" :Level 0.0))
    (Variables|Default|SetGE_ThermalVision _returnvalue_3)
    (Variables|Default|SetbThermalActive true)
    (else
      (if (Variables|Default|GetbThermalActive)
        (GameplayEffects|RemoveActiveGameplayEffect _returnvalue_2 (Variables|Default|GetGE_ThermalVision) -1)
        (Variables|Default|SetGE_ThermalVision (Utilities|Struct|MakeActiveGameplayEffectHandle))
        (Variables|Default|SetbThermalActive false)))))

(event Ability|EventOnEndAbility
  (params (bWasCancelled :type Boolean))
  ;[1 0.3607 0] CLEANUP - set default states, unbind events, call parent
  (Utilities|FlowControl|Sequence
    (:then_0
      (GameplayEffects|RemoveActiveGameplayEffect (Ability|GetFortAbilitySystemComponent (Ability|GetOwningActorfromActorInfo)) (Variables|Default|GetGE_ThermalVision) -1)
      (Variables|Default|SetGE_ThermalVision (Utilities|Struct|MakeActiveGameplayEffectHandle)))
    (:then_1
      (bind _currentthermalweapon (Variables|Default|GetCurrentThermalWeapon)
        (:then
          (EventDispatchers|UnbindEventfromOnTargetingChanged _currentthermalweapon (EventDispatchers|CreateEvent :!event Custom|OnWeaponTargetingChanged)))
        (:else)))
    (:then_2
      (bind _targetmodeselectorcomponent (Variables|Default|GetTargetModeSelectorComponent)
        (:then
          (EventDispatchers|UnbindEventfromOnTargetModeChanged _targetmodeselectorcomponent (EventDispatchers|CreateEvent :!event Custom|OnTargetModeChanged)))
        (:else)) ; This ability ends when the weapon is unequiped. \nWe want the component to stay around for the next time it is equiped so we don't destroy it here. \nIdeally this component would get cleaned up when the mod is removed, \nbut that functionality is not available yet
    )
    (:then_3
      (|CallParentFunction bWasCancelled))))

(fn OnRep_bIsTargeting
  (bind _bistargeting (Variables|Default|GetbIsTargeting))
  (bind _returnvalue (Math|Integer|SelectInt 22 8 _bistargeting))
  (Networking|CanExecuteCosmeticEvents
    (:True
      (bind _currentthermalweapon (Variables|Default|GetCurrentThermalWeapon)
        (:then
          (Rendering|SetCustomDepthStencilValue (Variables|/Script/FortniteGame.FortWeapon|GetWeaponMesh _currentthermalweapon) _returnvalue)
          (bind _returnvalue_2 (Class|FortWeaponModFunctionLibrary|GetRuntimeWeaponModMeshComponents _currentthermalweapon))
          (for _array_element _returnvalue_2
            (Rendering|SetCustomDepthStencilValue _array_element _returnvalue)))
        (:else)))))
