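# Sound cue for activating the toggleable thermal DMR scope optic in the WeaponMods content.
# A single wave player feeds a default Modulator, reusing the base game's Thermal Rifle scope zoom-out sample.
# It plays quietly on the weapon foley sound class, with overrides that disable attenuation and spatialization and prevent new instances at the concurrency limit.
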SoundCue /WeaponMods/Items/WeaponMods/Optics/ToggleThermal/Sound/ThermalDMR_Scope_Thermal_Activate_Cue
root=Modulator_1
nodes [
 {id=Modulator_1, children [WavePlayer_0]}
 {id=WavePlayer_0, properties {soundWaveAssetPtr=/Game/Sounds/Fort_Weapons/AutoRifle/Thermal_Rifle/ThermalRifle_Scope_ZoomOut01}}
]
properties {
 bOverrideAttenuation=true
 bOverrideConcurrency=true
 maxDistance=2097152
 soundClassObject=&Fort_WeaponFoley
 volumeMultiplier=0.15
 attenuationOverrides {bAttenuate=false, bSpatialize=false, falloffDistance=1500, nonSpatializedRadiusStart=200, attenuationShapeExtents {x=1200}}
 concurrencyOverrides {resolutionRule=PreventNew}
}
