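# Sound cue for switching off the thermal mode on the ThermalDMR scope, part of the toggle-thermal optic weapon mod.
# A random node picks one of two equally weighted deactivation wave variations.
# Attenuation overrides disable falloff and spatialization, so it plays flat at reduced volume in the weapon foley sound class.
# A concurrency override blocks new instances when the limit is reached rather than stopping playing ones.
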SoundCue /WeaponMods/Items/WeaponMods/Optics/ToggleThermal/Sound/ThermalDMR_Scope_Thermal_Deactivate_Cue
root=Random_1
nodes [
 {id=Random_1, children [WavePlayer_0,WavePlayer_3], properties {weights [1,1]}}
 {
  id=WavePlayer_0
  properties {soundWaveAssetPtr=&/ThermalDMR_Scope_Thermal_Deactivate_A_01}
 }
 {
  id=WavePlayer_3
  properties {soundWaveAssetPtr=&/ThermalDMR_Scope_Thermal_Deactivate_A_02}
 }
]
properties {
 bOverrideAttenuation=true
 bOverrideConcurrency=true
 maxDistance=2097152
 soundClassObject=&Fort_WeaponFoley
 volumeMultiplier=0.226
 attenuationOverrides {bAttenuate=false, bSpatialize=false, falloffDistance=1500, nonSpatializedRadiusStart=200, attenuationShapeExtents {x=1200}}
 concurrencyOverrides {resolutionRule=PreventNew}
}
