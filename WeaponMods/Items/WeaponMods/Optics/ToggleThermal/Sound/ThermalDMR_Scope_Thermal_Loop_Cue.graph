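# Sound cue for a thermal DMR scope in the ToggleThermal weapon mod, pairing an activation sound with a looping scope bed.
# A top mixer layers one of three randomly chosen activation waves, passed through a modulator, over a sub-mix of two looping beep tracks and a looping noise track.
# It uses the weapon foley sound class, overrides attenuation for distance falloff, and overrides concurrency to prevent new instances at the limit.
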SoundCue /WeaponMods/Items/WeaponMods/Optics/ToggleThermal/Sound/ThermalDMR_Scope_Thermal_Loop_Cue
root=Mixer_0
nodes [
 {id=Mixer_0, children [Modulator_1,Mixer_1], properties {inputVolume [1,1.2]}}
 {id=Mixer_1, children [WavePlayer_6,WavePlayer_7,WavePlayer_8], properties {inputVolume [0.8,0.8,0.5]}}
 {id=Modulator_1, children [Random_0], properties {pitchMax=1, pitchMin=1}}
 {id=Random_0, children [WavePlayer_1,WavePlayer_2,WavePlayer_3], properties {weights [1,1,1]}}
 {
  id=WavePlayer_1
  properties {soundWaveAssetPtr=&/ThermalDMR_Scope_Thermal_Activate_A_01}
 }
 {
  id=WavePlayer_2
  properties {soundWaveAssetPtr=&/ThermalDMR_Scope_Thermal_Activate_C_01}
 }
 {
  id=WavePlayer_3
  properties {soundWaveAssetPtr=&/ThermalDMR_Scope_Thermal_Activate_C_02}
 }
 {
  id=WavePlayer_6
  properties {bLooping=true, soundWaveAssetPtr=&/ThermalDMR_Scope_Loop_Beeps_02}
 }
 {
  id=WavePlayer_7
  properties {bLooping=true, soundWaveAssetPtr=&/ThermalDMR_Scope_Loop_Beeps_01}
 }
 {
  id=WavePlayer_8
  properties {bLooping=true, soundWaveAssetPtr=&/ThermalDMR_Scope_Loop_Noise_01}
 }
]
properties {
 bOverrideAttenuation=true
 bOverrideConcurrency=true
 maxDistance=2000
 soundClassObject=&Fort_WeaponFoley
 volumeMultiplier=0.226
 attenuationOverrides {falloffDistance=1500, nonSpatializedRadiusStart=400, stereoSpread=300, attenuationShapeExtents {x=500}}
 concurrencyOverrides {resolutionRule=PreventNew}
}
